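# M_DebugDecal is an unlit, translucent deferred decal material in the PCG resources, apparently for projecting a debug render target onto surfaces.
# A texture object parameter, SourceRenderTargetA, defaults to DebugDecalRenderTarget2D and feeds a linear-color texture sample, so the displayed target can be swapped per instance.
# Two scalar constants sit alongside it, and the graph shows no links from any node to the material outputs.
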Material /PCG/BP_Elements/Resources/MaterialsAndRTs/M_DebugDecal
graph {
 properties {
  blendMode=BLEND_Translucent
  materialDomain=MD_DeferredDecal
  pixelDepthOffsetMode=PDOM_Legacy
  shadingModel=MSM_Unlit
  usedShadingModels=Unlit
 }
 nodes [
  Constant()
  Constant() {r=0.5}
  n1 = TextureObjectParameter() {
   parameterName=SourceRenderTargetA
   samplerType=SAMPLERTYPE_LinearColor
   texture=&/DebugDecalRenderTarget2D
  }
  TextureSample(Tex=@n1) {samplerType=SAMPLERTYPE_LinearColor}
 ]
}
