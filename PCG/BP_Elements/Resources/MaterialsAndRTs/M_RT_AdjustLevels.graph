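# Unlit utility material in the PCG element resources that applies a levels adjustment to a source render target texture.
# It samples the SourceRenderTarget parameter as linear color, raises RGB to Power, then remaps it between Min and Max with a saturated inverse lerp.
# Offset is added to the clamped result before multiplying by Scale, and the value goes to Emissive Color through material attributes.
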Material /PCG/BP_Elements/Resources/MaterialsAndRTs/M_RT_AdjustLevels
graph {
 properties {bUseMaterialAttributes=true, pixelDepthOffsetMode=PDOM_Legacy, shadingModel=MSM_Unlit, usedShadingModels=Unlit}
 nodes [
  n1 = TextureObjectParameter() {parameterName=SourceRenderTarget, samplerType=SAMPLERTYPE_LinearColor, texture=/PCG/BP_Elements/Resources/Textures/T_BlackLinearFloat}
  n2 = TextureSample(Tex=@n1)
  n3 = ScalarParameter() {defaultValue=1, parameterName=Power}
  n4 = Saturate(None=InverseLinearInterpolate(A=ScalarParameter() {parameterName=Min}, B=ScalarParameter() {defaultValue=1, parameterName=Max}, Value=Power(Base=@n2.RGB, Exp=@n3)))
  n5 = ScalarParameter() {parameterName=Offset}
  SetMaterialAttributes(Emissive Color=Multiply(A=Add(A=@n4, B=@n5), B=ScalarParameter() {defaultValue=1, parameterName=Scale})) {attributeSetTypes [#1]}
 ]
}
