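# Unlit blur material in the PCG resources that samples a render target texture parameter, SourceRenderTarget, with a black linear float texture as default.
# It feeds that texture into the engine's SpiralBlur-Texture function, which averages samples along a spiral.
# A BlurDistance parameter, clamped to be non-negative, sets the blur radius, and a fixed high distance step count trades cost for smoothness.
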Material /PCG/BP_Elements/Resources/MaterialsAndRTs/M_RT_Blur
graph {
 properties {pixelDepthOffsetMode=PDOM_Legacy, shadingModel=MSM_Unlit, usedShadingModels=Unlit}
 nodes [
  n1 = TextureObjectParameter() {parameterName=SourceRenderTarget, samplerType=SAMPLERTYPE_LinearColor, texture=/PCG/BP_Elements/Resources/Textures/T_BlackLinearFloat}
  MaterialFunctionCall!(
   TextureObject=@n1
   UV
   Distance=Max(A=ScalarParameter() {defaultValue=0.05, parameterName=BlurDistance}, B=Constant() {r=0})
   Distance Steps=Constant() {r=64}
   Radial Steps
   Radial Offset
   Kernel Power
  ) {materialFunction=/Engine/Functions/Engine_MaterialFunctions01/Texturing/SpiralBlur-Texture}
 ]
}
