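# Unlit material that builds a binary color-key mask from a source render target texture parameter, likely for render target processing in PCG elements.
# It measures the RGB distance between each sampled pixel and the ColorKey parameter, then saturates and rounds it, so colors near the key become 0 and distant ones 1.
# The InvertMask scalar blends between that mask and its inverse.
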Material /PCG/BP_Elements/Resources/MaterialsAndRTs/M_RT_ColorMask
graph {
 properties {pixelDepthOffsetMode=PDOM_Legacy, shadingModel=MSM_Unlit, usedShadingModels=Unlit}
 nodes [
  n1 = TextureObjectParameter() {parameterName=SourceRenderTarget, samplerType=SAMPLERTYPE_LinearColor, texture=/PCG/BP_Elements/Resources/Textures/T_BlackLinearFloat}
  n2 = TextureSample(Tex=@n1)
  n3 = VectorParameter() {parameterName=ColorKey, defaultValue=<rgba 0.5776 0.7157 0.0908 1>}
  n4 = Round(None=Saturate(None=Distance(A=@n2.RGB, B=@n3.RGB)))
  LinearInterpolate(A=@n4, B=OneMinus(None=@n4), Alpha=ScalarParameter() {parameterName=InvertMask})
 ]
}
