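# M_RT_MultiplyByFloat is an unlit utility material in the PCG plugin's resources, built to scale render target contents by a scalar.
# It samples the SourceRenderTargetA texture parameter, which uses a linear color sampler and defaults to a black linear float texture.
# The sampled RGB is multiplied by the Multiplier scalar parameter, whose default leaves the values unchanged.
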Material /PCG/BP_Elements/Resources/MaterialsAndRTs/M_RT_MultiplyByFloat
graph {
 properties {pixelDepthOffsetMode=PDOM_Legacy, shadingModel=MSM_Unlit, usedShadingModels=Unlit}
 nodes [
  n1 = TextureObjectParameter() {parameterName=SourceRenderTargetA, samplerType=SAMPLERTYPE_LinearColor, texture=/PCG/BP_Elements/Resources/Textures/T_BlackLinearFloat}
  Multiply(A=TextureSample(Tex=@n1).RGB, B=ScalarParameter() {defaultValue=1, parameterName=Multiplier})
 ]
}
