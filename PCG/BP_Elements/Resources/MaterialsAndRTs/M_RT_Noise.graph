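# M_RT_Noise is an unlit material that layers procedural noise onto a source render target texture, apparently for drawing into render targets in the PCG element tools.
# Viewport UV divided by NoiseScale, with a constant Z, feeds a tiling Noise node, whose result is shifted by NoiseOffset and scaled by NoiseMultiplier.
# That noise is added to SourceRenderTarget, sampled at a fixed mip level with a black linear float default, and the sum is saturated.
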Material /PCG/BP_Elements/Resources/MaterialsAndRTs/M_RT_Noise
graph {
 properties {pixelDepthOffsetMode=PDOM_Legacy, shadingModel=MSM_Unlit, usedShadingModels=Unlit}
 nodes [
  n1 = Multiply(A=ScreenPosition().ViewportUV, B=Divide(A=Constant() {r=1}, B=ScalarParameter() {defaultValue=1, parameterName=NoiseScale}))
  n2 = Constant()
  n3 = Add(A=Noise!(World Position=AppendVector(A=@n1, B=@n2), FilterWidth) {bTiling=true}, B=ScalarParameter() {parameterName=NoiseOffset})
  n4 = ScalarParameter() {defaultValue=1, parameterName=NoiseMultiplier}
  n5 = TextureObjectParameter() {parameterName=SourceRenderTarget, samplerType=SAMPLERTYPE_LinearColor, texture=/PCG/BP_Elements/Resources/Textures/T_BlackLinearFloat}
  n6 = Constant()
  Saturate(None=Add(
    A=Multiply(A=@n3, B=@n4)
    B=TextureSample!(UVs, Tex=@n5, Level=@n6, Apply View MipBias) {mipValueMode=TMVM_MipLevel, samplerType=SAMPLERTYPE_LinearColor}.RGB
   ))
  TextureCoordinate()
 ]
}
