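# M_RT_NormalVariance is an unlit material that computes a normal variance mask from a world-normal texture, apparently for render target use.
# A spiral blur averages the WorldNormals texture, with distance, step counts, radial offset and kernel power exposed as parameters.
# The averaged vector's length shrinks where normals diverge, so one minus that length, raised to a fractional power and saturated, highlights areas of varying normals.
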Material /PCG/BP_Elements/Resources/MaterialsAndRTs/M_RT_NormalVariance
graph {
 properties {pixelDepthOffsetMode=PDOM_Legacy, shadingModel=MSM_Unlit, usedShadingModels=Unlit}
 nodes [
  n1 = TextureObjectParameter() {parameterName=WorldNormals, samplerType=SAMPLERTYPE_LinearColor, texture=/PCG/BP_Elements/Resources/Textures/T_WhiteLinearFloat}
  n2 = MaterialFunctionCall(Vector 3=MaterialFunctionCall!(
    TextureObject=@n1
    UV
    Distance=ScalarParameter() {defaultValue=0.002, parameterName=BlurDist}
    Distance Steps=ScalarParameter() {defaultValue=16, parameterName=DistanceSteps}
    Radial Steps=ScalarParameter() {defaultValue=16, parameterName=RadialSteps}
    Radial Offset=ScalarParameter() {defaultValue=0.62, parameterName=RadialOffset}
    Kernel Power=ScalarParameter() {defaultValue=1, parameterName=KernelPower}
   ) {materialFunction=/Engine/Functions/Engine_MaterialFunctions01/Texturing/SpiralBlur-Texture}.Result) {materialFunction=/Engine/Functions/Engine_MaterialFunctions02/Utility/VectorLength}
  Saturate(None=Power(Base=OneMinus(None=@n2.V3 Length)) {constExponent=0.2})
 ]
}
