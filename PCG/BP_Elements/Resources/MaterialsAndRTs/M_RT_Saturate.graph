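# Unlit utility material in the PCG resources that clamps a source render target's color to the 0 to 1 range.
# It samples the SourceRenderTargetA texture object parameter as linear color and passes the RGB through a Saturate node.
# The parameter defaults to a black linear float texture as a placeholder for the source render target.
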Material /PCG/BP_Elements/Resources/MaterialsAndRTs/M_RT_Saturate
graph {
 properties {pixelDepthOffsetMode=PDOM_Legacy, shadingModel=MSM_Unlit, usedShadingModels=Unlit}
 nodes [
  n1 = TextureObjectParameter() {parameterName=SourceRenderTargetA, samplerType=SAMPLERTYPE_LinearColor, texture=/PCG/BP_Elements/Resources/Textures/T_BlackLinearFloat}
  Saturate(None=TextureSample(Tex=@n1).RGB)
 ]
}
