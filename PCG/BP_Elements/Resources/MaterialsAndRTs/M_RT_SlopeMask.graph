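# Post-process, unlit material in the PCG resources that computes a slope mask from a world-normals texture, apparently for render targets.
# It dots the normals with an up vector, raises the result to a power, then remaps and clamps it between Min and Max.
# A lerp parameter blends toward the inverse, fully inverted by default.
# A SmoothStep of the same inputs sits alongside as a parallel branch.
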Material /PCG/BP_Elements/Resources/MaterialsAndRTs/M_RT_SlopeMask
graph {
 properties {materialDomain=MD_PostProcess, pixelDepthOffsetMode=PDOM_Legacy, shadingModel=MSM_Unlit, usedShadingModels=Unlit}
 nodes [
  n1 = ScalarParameter() {parameterName=Min}
  n2 = ScalarParameter() {defaultValue=0.7, parameterName=Max}
  n3 = TextureObjectParameter() {parameterName=WorldNormals, samplerType=SAMPLERTYPE_LinearColor, texture=/PCG/BP_Elements/Resources/Textures/T_WhiteLinearFloat}
  n4 = Power(
   Base=DotProduct(A=TextureSample(Tex=@n3).RGB, B=VectorParameter() {parameterName=UpVector, defaultValue=<rgba 0 0 1 1>}.RGB)
   Exp=ScalarParameter() {defaultValue=28, parameterName=VerticalMaskPow}
  )
  n5 = Saturate(None=InverseLinearInterpolate(A=@n1, B=@n2, Value=@n4))
  LinearInterpolate(A=@n5, B=OneMinus(None=@n5), Alpha=ScalarParameter() {defaultValue=1, parameterName=VerticalMaskInvert})
  MaterialFunctionCall(Alpha=@n4, Min=@n1, Max=@n2) {materialFunction=&SmoothStep}
 ]
}
