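# M_RT_Subtract is an unlit utility material that subtracts one render target's RGB from another's, apparently for render-target processing in the PCG elements.
# SourceRenderTargetA and SourceRenderTargetB are linear-color texture object parameters, both defaulting to a black linear float texture.
# Each is sampled at an explicit constant mip level, and B is subtracted from A with no clamp.
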Material /PCG/BP_Elements/Resources/MaterialsAndRTs/M_RT_Subtract
graph {
 properties {pixelDepthOffsetMode=PDOM_Legacy, shadingModel=MSM_Unlit, usedShadingModels=Unlit}
 nodes [
  n1 = TextureObjectParameter() {parameterName=SourceRenderTargetA, samplerType=SAMPLERTYPE_LinearColor, texture=/PCG/BP_Elements/Resources/Textures/T_BlackLinearFloat}
  n2 = TextureObjectParameter() {parameterName=SourceRenderTargetB, samplerType=SAMPLERTYPE_LinearColor, texture=&T_BlackLinearFloat}
  Subtract(
   A=TextureSample!(UVs, Tex=@n1, Level=Constant(), Apply View MipBias) {mipValueMode=TMVM_MipLevel}.RGB
   B=TextureSample(Tex=@n2, Level=Constant()) {mipValueMode=TMVM_MipLevel}.RGB
  )
 ]
}
