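# SceneDepthPP is a post-process material in the PCG BP_Elements resources, named for exposing scene depth.
# Its graph holds a single SceneTexture node set to sample scene depth, with no math or parameters.
# No link from that node to a material output is recorded, and no use of the material is defined here.
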Material /PCG/BP_Elements/Resources/MaterialsAndRTs/SceneDepthPP
graph {properties {materialDomain=MD_PostProcess, usedShadingModels=Default Lit}, nodes [SceneTexture() {sceneTextureId=PPI_SceneDepth}]}
